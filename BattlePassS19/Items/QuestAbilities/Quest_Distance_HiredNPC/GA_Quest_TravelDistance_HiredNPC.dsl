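; Gameplay ability for a Battle Pass S19 quest that tracks distance the player travels, presumably with a hired NPC.
; On activation it starts a looping timer with a random interval that adds the owning pawn's horizontal movement, floored to whole meters, to a cached total.
; Totals above 100 meters go to the quest manager as a tagged stat event, then reset.
; Ending the ability clears the timer and sends any remainder of at least 10 meters.
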(event Ability|EventActivateAbilityFromEvent
  (params (EventData :type Gameplay%20Event%20Data%20Structure%20%28by%20ref%29))
  ;[1 1 1] Activate ability - set vars needed to calculate and store off distance travelled
  (|CallParentFunction EventData)
  (bind _output_get (Variables|Default|SetCurrentLocation (Transformation|GetActorLocation (Variables|GAVars|Data|GetOwningPlayerPawn))))
  (Variables|Default|SetPreviousLocation _output_get)
  (bind _returnvalue (Utilities|Time|SetTimerbyEvent (delegate Custom|SendDistanceUpdate) (Math|Random|RandomFloatinRange 1.5 2.0) true))
  (Variables|Default|SetTimerHandle _returnvalue))

(event Custom|SendDistanceUpdate
  (bind (_previouslocation_x _previouslocation_y) (Math|Vector|BreakVector (Variables|GetPreviousLocation_X)))
  (bind (_currentlocation_x _currentlocation_y) (Math|Vector|BreakVector (Variables|GetCurrentLocation_X)))
  ;[1 1 1] Send Distance Update
  (Utilities|FlowControl|Sequence
    (:then_0
      ;;[1 1 1] Cache horizontal distance
      (Variables|Default|SetCurrentLocation (Transformation|GetActorLocation (Variables|GAVars|Data|GetOwningPlayerPawn)))
      (Variables|Default|SetMetersTravelled (+ (Math|Float|Floor (/ (Math|Vector|VectorLength (Math|Vector|vector-vector (Math|Vector|MakeVector _currentlocation_x _currentlocation_y) (Math|Vector|MakeVector _previouslocation_x _previouslocation_y))) 1e+02)) (Variables|Default|GetMetersTravelled)))
      (Variables|Default|SetPreviousLocation (Variables|Default|GetCurrentLocation)) ; Reset location
    )
    (:then_1
      ;;[1 1 1] If cached distance is greater than 100 meters, send quest update
      (if (> (Variables|Default|GetMetersTravelled) 100)
        (Quest|SendComplexCustomStatEvent (Variables|GAVars|Data|GetOwningPlayerQuestManager) 0 (Variables|Default|GetT_QuestSourceTags) "(GameplayTags=)" (Variables|Default|GetMetersTravelled))
        (Variables|Default|SetMetersTravelled 0)))))

(event Ability|EventOnEndAbility
  (params (bWasCancelled :type Boolean))
  (bind _timerhandle (Variables|Default|GetTImerHandle))
  ;[1 1 1] On End Ability
  (|CallParentFunction bWasCancelled)
  (if (Utilities|Time|IsValidTimerHandle _timerhandle)
    (Utilities|Time|ClearandInvalidateTimerbyHandle _timerhandle)
    (else
      _))
  (if (>= (Variables|Default|GetMetersTravelled) 10)
    ;[1 1 1] Send Distance Update
    ;;[1 1 1] If cached distance is greater than 100 meters, send quest update
    (Quest|SendComplexCustomStatEvent (Variables|GAVars|Data|GetOwningPlayerQuestManager) 0 (Variables|Default|GetT_QuestSourceTags) "(GameplayTags=)" (Variables|Default|GetMetersTravelled))
    (Variables|Default|SetMetersTravelled 0)))
